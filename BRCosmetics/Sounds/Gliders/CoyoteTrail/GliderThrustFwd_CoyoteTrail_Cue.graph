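# Sound cue for the forward-thrust sound of the CoyoteTrail glider cosmetic, using glider thrust attenuation and concurrency.
# A mixer layers one of three equally weighted thrust variants with a quieter random screech layer.
# The screech layer picks with replacement among four screeches and a heavily weighted empty slot, so it is silent more than half the time.
# A modulator at the root applies slight pitch variation to the whole mix.
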SoundCue /BRCosmetics/Sounds/Gliders/CoyoteTrail/GliderThrustFwd_CoyoteTrail_Cue
root=Modulator_0
nodes [
 {id=Mixer_0, children [Random_0,Random_1], properties {inputVolume [0.7,0.5]}}
 {id=Modulator_0, children [Mixer_0], properties {pitchMax=1.1, pitchMin=0.9}}
 {id=Random_0, children [WavePlayer_3,WavePlayer_4,WavePlayer_5], properties {weights [1,1,1]}}
 {
  id=Random_1
  children [WavePlayer_6,WavePlayer_7,WavePlayer_8,WavePlayer_9,null]
  properties {bRandomizeWithoutReplacement=false, weights [0.2,0.2,0.2,0.2,1]}
 }
 {id=WavePlayer_3, properties {soundWaveAssetPtr=&/Glider_Coyote_Trail_ThrustFwd_01}}
 {id=WavePlayer_4, properties {soundWaveAssetPtr=&/Glider_Coyote_Trail_ThrustFwd_02}}
 {id=WavePlayer_5, properties {soundWaveAssetPtr=&/Glider_Coyote_Trail_ThrustFwd_03}}
 {id=WavePlayer_6, properties {soundWaveAssetPtr=&/Glider_Coyote_Trail_Screech_01}}
 {id=WavePlayer_7, properties {soundWaveAssetPtr=&/Glider_Coyote_Trail_Screech_02}}
 {id=WavePlayer_8, properties {soundWaveAssetPtr=&/Glider_Coyote_Trail_Screech_03}}
 {id=WavePlayer_9, properties {soundWaveAssetPtr=&/Glider_Coyote_Trail_Screech_04}}
]
properties {
 attenuationSettings=&Fort_Player_Glider_Thrust_Athena_Att
 audioPropertiesSheet=&APS_Cosmetic_Glider_Loop
 maxDistance=12500
 soundClassObject=&Fort_Gliders_Land
 volumeMultiplier=1
 concurrencySet [&Athena_Glider_Thrust_Conc]
}
